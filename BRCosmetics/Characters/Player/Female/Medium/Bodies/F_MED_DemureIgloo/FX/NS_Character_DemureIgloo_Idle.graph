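# Idle Niagara effect for the DemureIgloo character cosmetic, rendering a socket-attached lamp from glow sprites, glass sprites, light meshes and real lights.
# Four emitters follow a system socket reader: FrontGLow, GlassLight, MainLight and Lights, the last using component and light renderers.
# System.EnableLight and its change tracking gate spawning and kill particles when off.
# A curve blends a waveform flicker into steady brightness, with MainLight using a separate front-end multiplier.
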NiagaraSystem /BRCosmetics/Characters/Player/Female/Medium/Bodies/F_MED_DemureIgloo/FX/NS_Character_DemureIgloo_Idle
dependencies {
 usedDataInterfaces [&NiagaraDataInterfaceCurve, &NiagaraDataInterfaceSocketReader]
 usedDynamicInputs [
  &FNTimeOfDayColor
  &Boolean_OrOperation
  &SetBoolByFloatComparison
 ]
 usedModules [
  &self:SystemSpawnScript.SystemScriptSource.SystemScriptGraph.NiagaraNodeAssignment_0.SetVariables_#1
  &SystemState
  &self:SystemSpawnScript.SystemScriptSource.SystemScriptGraph.NiagaraNodeAssignment_1.SetVariables_#2
  &self:SystemSpawnScript.SystemScriptSource.SystemScriptGraph.NiagaraNodeAssignment_2.SetVariables_#3
  &self:SystemSpawnScript.SystemScriptSource.SystemScriptGraph.NiagaraNodeAssignment_3.SetVariables_#4
  &self:SystemSpawnScript.SystemScriptSource.SystemScriptGraph.NiagaraNodeAssignment_4.SetVariables_#5
  /Game/Effects/Niagara/Modules/Logic/UpdateValueStateChange
 ]
}
emitters {
 FrontGLow {
  emitterUpdateScript [
   &EmitterState(Life Cycle Mode=::NewEnumerator1, Loop Behavior=::NewEnumerator1, Loop Duration Mode=::NewEnumerator1, UseLoopDelay=false)
   &SpawnBurst_Instantaneous(Use Spawn Probability=false, UseLoopCountLimit=true)
   &SpawnPerFrame(
    Spawn:NiagaraBool=&Boolean_AndOperation(
     Bool1=$System.EnableLight
     Bool2=&Boolean_AndOperation(
      Bool1=&Boolean_NotOperation(Bool=$System.UpdateValueStateChange.Changed)
      Bool2=&Boolean_NotOperation(Bool=$Transient.FirstFrame)
     )
    )
    Use Spawn Probability=false
   )
  ]
  particleSpawnScript [&InitializeParticle(
    Color=$System.Glow Color
    Sprite Size=&Lerp_Vector2(A=<40 40>, B=<10 10>, Alpha=&ReturnNormalizedExecIndex())
    Sprite Size Mode=::NewEnumerator4
    UsePositionOffset=false
    Write Material Random=true
   )]
  particleUpdateScript [
   &SolveForcesAndVelocity(Clamp Velocity=false, Limit Acceleration=false)
   &ParticleState(Kill Particles When Lifetime Has Elapsed=false)
   /SetVariables(Particles.ON_OFF:NiagaraFloat=&Lerp_Float(
     A=&Waveform("Waveform [1]"=::NewEnumerator0, "[1] Frequency"=5, "[1] Amplitude Min/Max"=<0 1>)
     B=1
     Alpha=&FloatFromCurve(FloatCurve=DataInterface() {propertyValues {Curve {keys [<at 0.1098 0>,<at 0.4762 1>]}}})
    )) {isSetParameters=true}
   &SocketLocation(
    SamplePosOffset=true
    Sampled Position Offset=&Lerp_Vector(A=<30 0 0>, B=<5 0 0>, Alpha=&ReturnNormalizedExecIndex())
    Socket Reader=$System.Socket Reader
    Socket Sampling Mode=::NewEnumerator3
    VelocitySpeedLimit=false
   ) {module=SocketLocation001}
   &ScaleColor(
    Scale Alpha=&Multiply_Float(A=$Particles.ON_OFF, B=&Lerp_Float(A=1, B=3, Alpha=&ReturnNormalizedExecIndex()))
    ScaleA=true
    ScaleRGB=true
   ) {module=ScaleColor001}
   &AlignSpriteToMeshOrientation(SetAlignment=true, SetFacing=true)
   &KillParticles(Kill Particles=&InvertBool(Boolean:NiagaraBool=$System.EnableLight))
  ]
  rendererClasses [&NiagaraSpriteRendererProperties]
  renderers [{rendererClass=&NiagaraSpriteRendererProperties}]
 }
 GlassLight {
  emitterUpdateScript [
   &EmitterState(Life Cycle Mode=::NewEnumerator1, Loop Behavior=::NewEnumerator1, Loop Duration Mode=::NewEnumerator1, UseLoopDelay=false)
   &SpawnBurst_Instantaneous(Use Spawn Probability=false, UseLoopCountLimit=true)
   &SpawnPerFrame(
    Spawn:NiagaraBool=&Boolean_AndOperation(
     Bool1=$System.EnableLight
     Bool2=&Boolean_AndOperation(
      Bool1=&Boolean_NotOperation(Bool=$System.UpdateValueStateChange.Changed)
      Bool2=&Boolean_NotOperation(Bool=$Transient.FirstFrame)
     )
    )
    Use Spawn Probability=false
   )
  ]
  particleSpawnScript [&InitializeParticle(Color=<rgba 3 3 3 0.1565>, Sprite Size=<4.43 3.5976>, Sprite Size Mode=::NewEnumerator4, UsePositionOffset=false, Write Material Random=true)]
  particleUpdateScript [
   &SolveForcesAndVelocity(Clamp Velocity=false, Limit Acceleration=false)
   &ParticleState(Kill Particles When Lifetime Has Elapsed=false)
   /SetVariables(Particles.ON_OFF:NiagaraFloat=&Lerp_Float(
     A=&Waveform("Waveform [1]"=::NewEnumerator0, "[1] Frequency"=5, "[1] Amplitude Min/Max"=<0 1>)
     B=1
     Alpha=&FloatFromCurve(FloatCurve=DataInterface() {propertyValues {Curve {keys [<at 0.1098 0>,<at 0.4762 1>]}}})
    )) {isSetParameters=true}
   &SocketLocation(SamplePosOffset=true, Socket Reader=$System.Socket Reader, Socket Sampling Mode=::NewEnumerator3, VelocitySpeedLimit=false) {module=SocketLocation001}
   &ScaleColor(Scale Alpha=$Particles.ON_OFF, ScaleA=true, ScaleRGB=true) {module=ScaleColor001}
   &AlignSpriteToMeshOrientation(SetAlignment=true, SetFacing=true)
   &KillParticles(Kill Particles=&InvertBool(Boolean:NiagaraBool=$System.EnableLight))
  ]
  rendererClasses [&NiagaraSpriteRendererProperties]
  renderers [
   {rendererClass=&NiagaraSpriteRendererProperties}
   {rendererClass=&NiagaraSpriteRendererProperties, rendererIndex=1}
  ]
 }
 Lights {
  emitterUpdateScript [
   &EmitterState(Life Cycle Mode=::NewEnumerator1, Loop Behavior=::NewEnumerator1, Loop Duration Mode=::NewEnumerator1, UseLoopDelay=false)
   &SpawnBurst_Instantaneous(Use Spawn Probability=false, UseLoopCountLimit=true)
   &SpawnPerFrame(
    Spawn:NiagaraBool=&Boolean_AndOperation(
     Bool1=$System.EnableLight
     Bool2=&Boolean_AndOperation(
      Bool1=&Boolean_NotOperation(Bool=$System.UpdateValueStateChange.Changed)
      Bool2=&Boolean_NotOperation(Bool=$Transient.FirstFrame)
     )
    )
    Use Spawn Probability=false
   )
  ]
  particleSpawnScript [
   &InitializeParticle(Color Mode=::NewEnumerator0, UsePositionOffset=false, Write Material Random=true)
   &InitialMeshOrientation(Mesh Axis To Orient=<0 0 -1>, Mesh Orientation Mode=::NewEnumerator2, Orientation Vector=<0 1 0>, Use Rotation Vector=false)
  ]
  particleUpdateScript [
   &SocketLocation(
    SamplePosOffset=true
    Sampled Position Offset=<25 0 0>
    Socket Reader=$System.Socket Reader
    Socket Sampling Mode=::NewEnumerator3
    VelocitySpeedLimit=false
   ) {module=SocketLocation001}
   /SetVariables(Particles.ON_OFF:NiagaraFloat=&Lerp_Float(
     A=&Waveform("Waveform [1]"=::NewEnumerator0, "[1] Frequency"=5, "[1] Amplitude Min/Max"=<0 1>)
     B=1
     Alpha=&FloatFromCurve(FloatCurve=DataInterface() {propertyValues {Curve {keys [<at 0.1098 0>,<at 0.4762 1>]}}})
    )) {isSetParameters=true}
   /SetVariables(Particles.PointLightPosition:NiagaraPosition=$Output.SocketLocation001.SampledPosition) {isSetParameters=true}
   &SocketLocation(SamplePosOffset=true, Socket Reader=$System.Socket Reader, Socket Sampling Mode=::NewEnumerator3, VelocitySpeedLimit=false)
   &ParticleState(Kill Particles When Lifetime Has Elapsed=false)
   &FN_Light_Attributes(
    Color=<rgba 0.9635 0.9635 0.9635 1>
    Create Color Attribute=true
    Create Radius Attribute=true
    Enable Particle Light Enable=true
    Light Enabled=true
    Radius=175
    Scale Color=$Particles.ON_OFF
   )
   &FN_Light_Attributes(
    Create Color Attribute=true
    Create Exponent Attribute=true
    Create Radius Attribute=true
    Enable Particle Light Enable=true
    Light Enabled=true
    Radius=40
    Scale Color=&Multiply_Float(A=$Particles.ON_OFF, B=20)
   ) {module=FN_Light_Attributes001}
   &KillParticles(Kill Particles=&InvertBool(Boolean:NiagaraBool=$System.EnableLight))
  ]
  rendererClasses [&NiagaraComponentRendererProperties, &NiagaraLightRendererProperties]
  renderers [
   {rendererClass=&NiagaraComponentRendererProperties, rendererIndex=0}
   {rendererClass=&NiagaraLightRendererProperties, rendererIndex=1}
  ]
 }
 MainLight {
  emitterUpdateScript [
   &EmitterState()
   &SpawnBurst_Instantaneous(Use Spawn Probability=false, UseLoopCountLimit=true)
   &SpawnBurst_Instantaneous(Spawn Group=1, Use Spawn Probability=false, UseLoopCountLimit=true) {module=SpawnBurst_Instantaneous001}
   &SpawnPerFrame(
    Spawn:NiagaraBool=&Boolean_AndOperation(
     Bool1=$System.EnableLight
     Bool2=&Boolean_AndOperation(
      Bool1=&Boolean_NotOperation(Bool=$System.UpdateValueStateChange.Changed)
      Bool2=&Boolean_NotOperation(Bool=$Transient.FirstFrame)
     )
    )
    Use Spawn Probability=false
   )
   &SpawnPerFrame(
    Spawn:NiagaraBool=&Boolean_AndOperation(
     Bool1=$System.EnableLight
     Bool2=&Boolean_AndOperation(
      Bool1=&Boolean_NotOperation(Bool=$System.UpdateValueStateChange.Changed)
      Bool2=&Boolean_NotOperation(Bool=$Transient.FirstFrame)
     )
    )
    Spawn Group=1
    Use Spawn Probability=false
   ) {module=SpawnPerFrame001}
  ]
  particleSpawnScript [
   &InitializeParticle(Color=<rgba 30 6.9446 0 0.01>, Mesh Scale Mode=::NewEnumerator3, Mesh Uniform Scale=0.9, UsePositionOffset=false, Write Material Random=true)
   &InitialMeshOrientation(Mesh Orientation Mode=::NewEnumerator2, Orientation Vector=<0 1 0>, Use Rotation Vector=true)
   /SetVariables(
    Particles.FlareColor=<rgba 3 3 3 1>
    Particles.GlowColor=&Multiply_LinearColorByFloat(Linear Color=$System.Faint Glow Color, Float=0.8)
    Particles.LightMeshColor:LinearColor=<rgba 3 3 3 0.4>
    Particles.LightMeshOuterColor:LinearColor=<rgba 1 1 1 0.9>
   ) {isSetParameters=true}
   /SetVariables(Particles.FlareSize:Vector2f=<10 10>, Particles.GlowSize=<50 50>, Particles.LightMeshScale:Vector3f=<0.3 0.35 0.35>) {isSetParameters=true}
   /SetVariables(
    Particles.FlarePosition:NiagaraPosition=&AddVectorToPosition(Position=$Particles.Position, Vector=<0 0 -2.6807>)
    Particles.GlowPosition:NiagaraPosition=<0 25 0>
    Particles.LightMeshPosition:NiagaraPosition=&AddVectorToPosition(Position=$Particles.Position, Vector=<0 0 -3.2785>)
   ) {isSetParameters=true}
   /SetVariables(Particles.FlareOffset=2) {isSetParameters=true}
   /SetVariables(Particles.FrontEndMult=0.75) {isSetParameters=true}
   /SetVariables(Particles.VisibilityTag=$Emitter.SpawnGroup) {isSetParameters=true}
  ]
  particleUpdateScript [
   &ParticleState(Kill Particles When Lifetime Has Elapsed=false)
   &SocketLocation(
    SamplePosOffset=true
    Sampled Position Offset=/Niagara/DynamicInputs/Helpers/MaskPosition(
     Mask Source:NiagaraInt32=0
     Mask Parameter:NiagaraInt32=?
     Mask:NiagaraInt32=0
     Masked Position:NiagaraPosition=<0 0 0>
     Passthrough Position:NiagaraPosition=<16 0 0>
    )
    Socket Reader=$System.Socket Reader
    Socket Sampling Mode=::NewEnumerator3
    VelocitySpeedLimit=false
   ) {module=SocketLocation001}
   /SetVariables(
    Particles.FlarePosition:NiagaraPosition=&AddVectorToPosition(Position=$Particles.Position, Vector=<0 0.75 0>)
    Particles.GlowPosition:NiagaraPosition=&Add_Vector(A=Linked() {linkedVariable {name=Particles.Position, type=NiagaraPosition}}, B=<0 0 0>)
    Particles.LightMeshPosition:NiagaraPosition=&AddVectorToPosition(Position=$Particles.Position, Vector=<0 0 0>)
   ) {isSetParameters=true}
   /SetVariables(Particles.ON_OFF:NiagaraFloat=&Lerp_Float(
     A=&Waveform("Waveform [1]"=::NewEnumerator0, "[1] Frequency"=5, "[1] Amplitude Min/Max"=<0 1>)
     B=1
     Alpha=&FloatFromCurve(FloatCurve=DataInterface() {propertyValues {Curve {keys [<at 0.1098 0>,<at 0.4762 1>]}}})
    )) {isSetParameters=true}
   /SetVariables(Particles.Flicker=1) {isSetParameters=true}
   /SetVariables(
    Particles.Color=&Multiply_LinearColorByFloat(Linear Color=$Particles.Initial.Color, Float=$Particles.Flicker)
    Particles.FlareColor=&Multiply_LinearColorByFloat(Linear Color=$Particles.Initial.FlareColor, Float=$Particles.Flicker)
    Particles.GlowColor=&Multiply_LinearColorByFloat(Linear Color=$Particles.Initial.GlowColor, Float=$Particles.Flicker)
    Particles.LightMeshColor:LinearColor=&Multiply_LinearColorByFloat(Linear Color=$Particles.Initial.LightMeshColor, Float=$Particles.Flicker)
    Particles.LightMeshOuterColor:LinearColor=&Multiply_LinearColorByFloat(Linear Color=$Particles.Initial.LightMeshOuterColor, Float=$Particles.Flicker)
   ) {isSetParameters=true}
   /SetVariables(
    Particles.Color=&Multiply_LinearColorByFloat(
     Linear Color=$Particles.Color
     Float=&FN_FrontEndFloatModifier(Gameplay=$Particles.ON_OFF, Front End=$Particles.FrontEndMult)
    )
    Particles.FlareColor=&Multiply_LinearColorByFloat(
     Linear Color=$Particles.FlareColor
     Float=&FN_FrontEndFloatModifier(Gameplay=$Particles.ON_OFF, Front End=$Particles.FrontEndMult)
    )
    Particles.GlowColor=&Multiply_LinearColorByFloat(
     Linear Color=$Particles.GlowColor
     Float=&FN_FrontEndFloatModifier(Gameplay=$Particles.ON_OFF, Front End=$Particles.FrontEndMult)
    )
    Particles.LightMeshColor:LinearColor=&Multiply_LinearColorByFloat(
     Linear Color=$Particles.LightMeshColor
     Float=&FN_FrontEndFloatModifier(Gameplay=$Particles.ON_OFF, Front End=$Particles.FrontEndMult)
    )
    Particles.LightMeshOuterColor:LinearColor=&Multiply_LinearColorByFloat(
     Linear Color=$Particles.LightMeshOuterColor
     Float=&FN_FrontEndFloatModifier(Gameplay=$Particles.ON_OFF, Front End=$Particles.FrontEndMult)
    )
   ) {isSetParameters=true}
   &KillParticles(Kill Particles=&InvertBool(Boolean:NiagaraBool=$System.EnableLight))
  ]
  rendererClasses [&NiagaraMeshRendererProperties, &NiagaraSpriteRendererProperties]
  renderers [
   {rendererClass=&NiagaraMeshRendererProperties, rendererIndex=0}
   {rendererClass=&NiagaraMeshRendererProperties, rendererIndex=1}
   {rendererClass=&NiagaraMeshRendererProperties, rendererIndex=2}
   {rendererClass=&NiagaraSpriteRendererProperties, rendererIndex=3}
   {rendererClass=&NiagaraSpriteRendererProperties, rendererIndex=4}
   {rendererClass=&NiagaraMeshRendererProperties, rendererIndex=5}
  ]
 }
}
summary {
 systemName=NS_Character_DemureIgloo_Idle
 emitters [
  {emitterName=FrontGLow, rendererClasses [&NiagaraSpriteRendererProperties]}
  {emitterName=MainLight, rendererClasses [&NiagaraMeshRendererProperties, &NiagaraSpriteRendererProperties]}
  {emitterName=GlassLight, rendererClasses [&NiagaraSpriteRendererProperties]}
  {emitterName=Lights, rendererClasses [&NiagaraComponentRendererProperties, &NiagaraLightRendererProperties]}
 ]
 userVariables [
  User.bIsFrontEnd {type=NiagaraBool}
  User.bIsFrontEndPreview {type=NiagaraBool}
  User.bIsHeadHidden {type=NiagaraBool}
  User.bIsPrimaryPreviewItem {type=NiagaraBool}
 ]
}
