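# Material function for the DemureIgloo character body that adds a parallax flashlight glow and time-of-day scaling to emissive.
# A camera-vector parallax offset moves UVs under a small sphere mask, clipped by a vertical UV band, to shape the glow.
# A UV-half mask gated by isFaceAcc is declared for blending.
# Scalar_Time_of_Day scales the result, with morning, day and evening values lerped toward zero by a mask while night stays unmasked.
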MaterialFunction /BRCosmetics/Characters/Player/Female/Medium/Bodies/F_MED_DemureIgloo/Materials/MF_DemureIglooTODEmissive
graph {nodes [
  n1 = GetMaterialAttributes(None=FunctionInput() {inputType=FunctionInput_MaterialAttributes}) {attributeGetTypes [#1]}
  n2 = TextureCoordinate() {vTiling=0.7}
  n3 = Normalize(VectorInput=Subtract(A=CameraPositionWS(), B=WorldPosition() {worldPositionShaderOffset=WPT_ExcludeAllShaderOffsets}.XYZ))
  n4 = Multiply(
   A=ComponentMask(None=Transform(None=@n3) {transformSourceType=TRANSFORMSOURCE_World, transformType=TRANSFORM_Tangent}) {g=true, r=true}
   B=ScalarParameter() {parameterName=Parallax Intensity}
  )
  n5 = ScalarParameter() {defaultValue=0.884, group=FlashlightParallax, parameterName=FlashlightOffsetX}
  n6 = ScalarParameter() {defaultValue=0.3293, group=FlashlightParallax, parameterName=FlashlightOffsetY}
  n7 = Power(
   Base=SphereMask(A=Subtract(A=@n2, B=@n4), B=AppendVector(A=@n5, B=@n6), Radius=Constant() {r=0.04}, Hardness=Constant())
   Exp=ScalarParameter() {defaultValue=8, parameterName=Falloff}
  )
  n8 = TextureCoordinate()
  n9 = ScalarParameter() {defaultValue=18, group=FlashlightParallax, parameterName=MaskSize}
  n10 = ComponentMask(None=Add(A=Multiply(A=@n8, B=@n9) {constB=18}, B=ScalarParameter() {defaultValue=-7.82, group=FlashlightParallax, parameterName=MaskAdd})) {g=true}
  n11 = Multiply(A=@n10, B=OneMinus(None=@n10))
  n12 = ScalarParameter() {defaultValue=2, group=FlashlightParallax, parameterName=MaskPow}
  n13 = Saturate(None=Multiply(A=Power(Base=@n11, Exp=@n12)) {constB=50})
  n14 = LinearInterpolate(
   A=NamedRerouteDeclaration(None=@n1.EmissiveColor) {name=EmissiveColor, nodeColor=<rgba 0.9219 0 1 1>}
   B=Multiply(A=NamedRerouteUsage(), B=Multiply(A=@n7, B=@n13))
   Alpha=NamedRerouteUsage()
  ) {constAlpha=1}
  n15 = NamedRerouteUsage()
  n16 = MaterialFunctionCall(
   Value_Morning=LinearInterpolate(A=ScalarParameter() {defaultValue=1, group=TOD_Emissive, parameterName=Value_Morning}, Alpha=@n15) {constB=0}
   Value_Day=LinearInterpolate(A=ScalarParameter() {defaultValue=1, group=TOD_Emissive, parameterName=Value_Day}, Alpha=@n15) {constB=0}
   Value_Evening=LinearInterpolate(A=ScalarParameter() {defaultValue=1, group=TOD_Emissive, parameterName=Value_Evening}, Alpha=@n15) {constB=0}
   Value_Night=ScalarParameter() {defaultValue=3, group=TOD_Emissive, parameterName=Value_Night}
  ) {materialFunction=&Scalar_Time_of_Day}
  FunctionOutput(None=SetMaterialAttributes!(MaterialAttributes=@n1.MaterialAttributes, EmissiveColor=Multiply(A=@n14, B=@n16.Result)) {attributeSetTypes [#1]})
  n17 = TextureCoordinate()
  n18 = OneMinus(None=Step(Y=ComponentMask(None=@n17) {r=true}) {constX=0.5})
  NamedRerouteDeclaration(None=Multiply(A=Saturate(None=@n18), B=ScalarParameter() {parameterName=isFaceAcc})) {name=CameraEmissiveMask, nodeColor=<rgba 0.7578 1 0 1>}
  Saturate(None=@n7)
 ]}
