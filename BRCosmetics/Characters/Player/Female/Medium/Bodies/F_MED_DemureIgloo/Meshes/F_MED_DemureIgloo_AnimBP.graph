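# Post-process style cosmetic AnimBlueprint for the F_MED_DemureIgloo female body, adding secondary motion to its hook, tubes and ropes.
# It copies the attached parent mesh's pose, runs a body Control Rig and caches the result.
# Weighted multi-bone constraints position hook, tube and right-arm twist-fix bones, interleaved with root-relative, angle-limited AnimDynamics on hook, tube and left-thigh bones.
# An Aim Control Rig finishes the pose; the EventGraph has no logic.
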AnimBlueprint /BRCosmetics/Characters/Player/Female/Medium/Bodies/F_MED_DemureIgloo/Meshes/F_MED_DemureIgloo_AnimBP
target_skeleton=&/F_MED_DemureIgloo_Skeleton
graphs [
 AnimGraph {nodes [
   n1 = "Constraint - dyn_hook (PPP)"!(
    ComponentPose=Local To Component(LocalPose=Use cached pose 'StartPose'())
    Alpha=1
    ConstraintWeights_0=0.8053
    ConstraintWeights_1:float=0.1573
    ConstraintWeights_2:float=0.0375
   ) -> (Pose) {
    boneToModify {boneName=dyn_hook}
    constraintSetup [{targetBone {boneName=pelvis}}, {targetBone {boneName=deform_glute_r}}, {targetBone {boneName=thigh_twist_01_r}}]
    constraintWeights [1,1,1]
   }
   n2 = Anim Dynamics - Bone: dyn_hook!(ComponentPose=@n1, Alpha=1, GravityOverride:Vector="0,0,-980") -> (Pose) {
    angularDampingOverride=0.999
    angularSpringConstant=800
    bOverrideAngularDamping=true
    bUseGravityOverride=true
    gravityScale=0.3
    linearSpringConstant=800
    simulationSpace=RootRelative
    alphaBoolBlend {blendOption=Linear}
    boundBone {boneName=dyn_hook}
    componentAppliedLinearAccClamp=<2000 2000 1600>
    componentLinearAccScale=<0.4 0.4 0.4>
    componentLinearVelScale=<0.45 0.45 0.55>
    gravityOverride {z=-980}
    physicsBodyDefinitions [{
      sphereCollisionRadius=2
      boundBone {boneName=dyn_hook}
      boxExtents=<20 20 20>
      constraintSetup {angularLimitsMax=<0 0 20>, angularLimitsMin=<0 -60 -20>, angularTarget=<0 0 0>}
      localJointOffset=<5 0 -1>
     }]
    relativeSpaceBone {boneName=root}
   }
   n3 = "Constraint - upperarm_twist_fix_01 (PPPPPPP)"!(
    ComponentPose=@n2
    Alpha=1
    ConstraintWeights_0=0.451
    ConstraintWeights_1:float=0.2123
    ConstraintWeights_2:float=0.1651
    ConstraintWeights_3:float=0.1052
    ConstraintWeights_4:float=0.0508
    ConstraintWeights_5:float=0.0154
    ConstraintWeights_6:float=0.0002
   ) -> (Pose) {
    boneToModify {boneName=upperarm_twist_fix_01}
    constraintSetup [
     {targetBone {boneName=spine_05}}
     {targetBone {boneName=upperarm_r}}
     {targetBone {boneName=upperarm_twist_01_r}}
     {targetBone {boneName=clavicle_r}}
     {targetBone {boneName=deform_upperarm_bk_r}}
     {targetBone {boneName=upperarm_twist_02_r}}
     {targetBone {boneName=deform_upperarm_up_FIx}}
    ]
    constraintWeights [1,1,1,1,1,1,1]
   }
   n4 = "Constraint - upperarm_twist_fix_02 (PPPPPP)"!(
    ComponentPose=@n3
    Alpha=1
    ConstraintWeights_0=0.3515
    ConstraintWeights_1:float=0.3004
    ConstraintWeights_2:float=0.1623
    ConstraintWeights_3:float=0.1113
    ConstraintWeights_4:float=0.0394
    ConstraintWeights_5:float=0.0351
   ) -> (Pose) {
    boneToModify {boneName=upperarm_twist_fix_02}
    constraintSetup [
     {targetBone {boneName=spine_05}}
     {targetBone {boneName=upperarm_twist_01_r}}
     {targetBone {boneName=upperarm_r}}
     {targetBone {boneName=deform_upperarm_bk_r}}
     {targetBone {boneName=clavicle_r}}
     {targetBone {boneName=upperarm_twist_02_r}}
    ]
    constraintWeights [1,1,1,1,1,1]
   }
   n5 = "Constraint - upperarm_twist_fix_03 (PPPPPP)"!(
    ComponentPose=@n4
    Alpha=1
    ConstraintWeights_0=0.433
    ConstraintWeights_1:float=0.2418
    ConstraintWeights_2:float=0.1309
    ConstraintWeights_3:float=0.0916
    ConstraintWeights_4:float=0.0752
    ConstraintWeights_5:float=0.0274
   ) -> (Pose) {
    boneToModify {boneName=upperarm_twist_fix_03}
    constraintSetup [
     {targetBone {boneName=upperarm_twist_01_r}}
     {targetBone {boneName=spine_05}}
     {targetBone {boneName=upperarm_r}}
     {targetBone {boneName=deform_upperarm_bk_r}}
     {targetBone {boneName=upperarm_twist_02_r}}
     {targetBone {boneName=clavicle_r}}
    ]
    constraintWeights [1,1,1,1,1,1]
   }
   n6 = "Constraint - dyn_tube_base (PPPPP)"!(
    ComponentPose=@n5
    Alpha=1
    ConstraintWeights_0=0.9037
    ConstraintWeights_1:float=0.0782
    ConstraintWeights_2:float=0.0152
    ConstraintWeights_3:float=0.0015
    ConstraintWeights_4:float=0.0014
   ) -> (Pose) {
    boneToModify {boneName=dyn_tube_base}
    constraintSetup [
     {targetBone {boneName=spine_05}}
     {targetBone {boneName=spine_04}}
     {targetBone {boneName=clavicle_r}}
     {targetBone {boneName=spine_01}}
     {targetBone {boneName=spine_02}}
    ]
    constraintWeights [1,1,1,1,1]
   }
   n7 = "Constraint - dyn_tube_01 (PPP)"!(ComponentPose=@n6, Alpha=1, ConstraintWeights_0=0.7694, ConstraintWeights_1:float=0.1611, ConstraintWeights_2:float=0.0695) -> (Pose) {
    boneToModify {boneName=dyn_tube_01}
    constraintSetup [{targetBone {boneName=spine_05}}, {targetBone {boneName=upperarm_twist_01_r}}, {targetBone {boneName=clavicle_r}}]
    constraintWeights [1,1,1]
   }
   n8 = "Constraint - dyn_tube_01_a (PPPPP)"!(
    ComponentPose=@n7
    Alpha=1
    ConstraintWeights_0=0.4792
    ConstraintWeights_1:float=0.2227
    ConstraintWeights_2:float=0.196
    ConstraintWeights_3:float=0.0956
    ConstraintWeights_4:float=0.0065
   ) -> (Pose) {
    boneToModify {boneName=dyn_tube_01_a}
    constraintSetup [
     {targetBone {boneName=spine_05}}
     {targetBone {boneName=spine_04}}
     {targetBone {boneName=spine_02}}
     {targetBone {boneName=spine_01}}
     {targetBone {boneName=clavicle_r}}
    ]
    constraintWeights [1,1,1,1,1]
   }
   n9 = "Constraint - dyn_tube_01_b (PPPPPP)"!(
    ComponentPose=@n8
    Alpha=1
    ConstraintWeights_0=0.4778
    ConstraintWeights_1:float=0.2026
    ConstraintWeights_2:float=0.1963
    ConstraintWeights_3:float=0.1005
    ConstraintWeights_4:float=0.0149
    ConstraintWeights_5:float=0.0078
   ) -> (Pose) {
    boneToModify {boneName=dyn_tube_01_b}
    constraintSetup [
     {targetBone {boneName=spine_05}}
     {targetBone {boneName=spine_02}}
     {targetBone {boneName=spine_04}}
     {targetBone {boneName=spine_01}}
     {targetBone {boneName=clavicle_r}}
     {targetBone {boneName=pelvis}}
    ]
    constraintWeights [1,1,1,1,1,1]
   }
   n10 = "Constraint - dyn_tube_02 (PPPPPP)"!(
    ComponentPose=@n9
    Alpha=1
    ConstraintWeights_0=0.3675
    ConstraintWeights_1:float=0.1725
    ConstraintWeights_2:float=0.1603
    ConstraintWeights_3:float=0.1554
    ConstraintWeights_4:float=0.0883
    ConstraintWeights_5:float=0.056
   ) -> (Pose) {
    boneToModify {boneName=dyn_tube_02}
    constraintSetup [
     {targetBone {boneName=spine_05}}
     {targetBone {boneName=spine_01}}
     {targetBone {boneName=spine_02}}
     {targetBone {boneName=pelvis}}
     {targetBone {boneName=spine_04}}
     {targetBone {boneName=clavicle_r}}
    ]
    constraintWeights [1,1,1,1,1,1]
   }
   n11 = "Constraint - dyn_tube_02_a (PPPPPP)"!(
    ComponentPose=@n10
    Alpha=1
    ConstraintWeights_0=0.3427
    ConstraintWeights_1:float=0.1986
    ConstraintWeights_2:float=0.1682
    ConstraintWeights_3:float=0.131
    ConstraintWeights_4:float=0.1261
    ConstraintWeights_5:float=0.0334
   ) -> (Pose) {
    boneToModify {boneName=dyn_tube_02_a}
    constraintSetup [
     {targetBone {boneName=spine_05}}
     {targetBone {boneName=spine_01}}
     {targetBone {boneName=spine_02}}
     {targetBone {boneName=spine_04}}
     {targetBone {boneName=pelvis}}
     {targetBone {boneName=clavicle_r}}
    ]
    constraintWeights [1,1,1,1,1,1]
   }
   n12 = "Constraint - dyn_tube_02_b (PPPPPP)"!(
    ComponentPose=@n11
    Alpha=1
    ConstraintWeights_0=0.3026
    ConstraintWeights_1:float=0.2291
    ConstraintWeights_2:float=0.1829
    ConstraintWeights_3:float=0.1394
    ConstraintWeights_4:float=0.121
    ConstraintWeights_5:float=0.025
   ) -> (Pose) {
    boneToModify {boneName=dyn_tube_02_b}
    constraintSetup [
     {targetBone {boneName=spine_05}}
     {targetBone {boneName=spine_02}}
     {targetBone {boneName=spine_01}}
     {targetBone {boneName=spine_04}}
     {targetBone {boneName=pelvis}}
     {targetBone {boneName=clavicle_r}}
    ]
    constraintWeights [1,1,1,1,1,1]
   }
   n13 = Anim Dynamics - Bone: dyn_tube_01!(ComponentPose=@n12, Alpha=1) -> (Pose) {
    angularDampingOverride=1
    angularSpringConstant=5000
    bAngularSpring=true
    bOverrideAngularDamping=true
    bUseGravityOverride=true
    gravityScale=0.3
    linearSpringConstant=800
    simulationSpace=RootRelative
    alphaBoolBlend {blendOption=Linear}
    boundBone {boneName=dyn_tube_01}
    componentAppliedLinearAccClamp=<1000 1000 1000>
    componentLinearAccScale=<0.3 0.3 0.3>
    componentLinearVelScale=<0.35 0.35 0.35>
    gravityOverride {z=-750}
    physicsBodyDefinitions [{
      sphereCollisionRadius=2
      boundBone {boneName=dyn_tube_01}
      boxExtents=<15 15 15>
      constraintSetup {angularLimitsMax=<1.5 4 -3>, angularLimitsMin=<-1.5 -4 -20>, angularTarget=<0 0 0>}
      localJointOffset=<5 -0.5 1>
     }]
    relativeSpaceBone {boneName=root}
    simSpaceSettings {maxAngularAcceleration=500, maxAngularVelocity=50, externalAngularVelocity {x=10}}
   }
   n14 = Anim Dynamics - Bone: dyn_tube_02!(ComponentPose=@n13, Alpha=1) -> (Pose) {
    angularDampingOverride=1
    angularSpringConstant=5000
    bAngularSpring=true
    bOverrideAngularDamping=true
    bUseGravityOverride=true
    gravityScale=0.3
    linearSpringConstant=800
    simulationSpace=RootRelative
    alphaBoolBlend {blendOption=Linear}
    boundBone {boneName=dyn_tube_02}
    componentAppliedLinearAccClamp=<1000 1000 1000>
    componentLinearAccScale=<0.3 0.3 0.3>
    componentLinearVelScale=<0.35 0.35 0.35>
    gravityOverride {z=-750}
    physicsBodyDefinitions [{
      sphereCollisionRadius=2
      boundBone {boneName=dyn_tube_02}
      boxExtents=<20 20 20>
      constraintSetup {angularLimitsMax=<1.5 4 -3>, angularLimitsMin=<-1.5 -4 -20>, angularTarget=<0 0 0>}
      localJointOffset=<5 -0.5 1>
     }]
    relativeSpaceBone {boneName=root}
    simSpaceSettings {maxAngularAcceleration=500, maxAngularVelocity=50, externalAngularVelocity {x=10}}
   }
   n15 = Anim Dynamics - Bone: dyn_thigh_l_tube_01!(ComponentPose=@n14, Alpha=1) -> (Pose) {
    angularDampingOverride=0.999
    angularSpringConstant=5000
    bAngularSpring=true
    bOverrideAngularDamping=true
    bUseGravityOverride=true
    gravityScale=0.3
    linearSpringConstant=800
    simulationSpace=RootRelative
    alphaBoolBlend {blendOption=Linear}
    boundBone {boneName=dyn_thigh_l_tube_01}
    componentAppliedLinearAccClamp=<1200 1200 1200>
    componentLinearAccScale=<0.3 0.3 0.3>
    componentLinearVelScale=<0.35 0.35 0.35>
    gravityOverride {z=-980}
    physicsBodyDefinitions [{
      sphereCollisionRadius=2
      boundBone {boneName=dyn_thigh_l_tube_01}
      boxExtents=<25 25 25>
      constraintSetup {angularLimitsMax=<1.5 4 0>, angularLimitsMin=<-1.5 -4 -7.5>, angularTarget=<0 0 0>}
      localJointOffset=<5 0 -1>
     }]
    relativeSpaceBone {boneName=root}
    simSpaceSettings {maxAngularAcceleration=500, maxAngularVelocity=50}
   }
   n16 = Anim Dynamics - Bone: dyn_thigh_l_tube_02!(ComponentPose=@n15, Alpha=1) -> (Pose) {
    angularDampingOverride=0.999
    angularSpringConstant=5000
    bAngularSpring=true
    bOverrideAngularDamping=true
    bUseGravityOverride=true
    gravityScale=0.3
    linearSpringConstant=800
    simulationSpace=RootRelative
    alphaBoolBlend {blendOption=Linear}
    boundBone {boneName=dyn_thigh_l_tube_02}
    componentAppliedLinearAccClamp=<1200 1200 1200>
    componentLinearAccScale=<0.3 0.3 0.3>
    componentLinearVelScale=<0.35 0.35 0.35>
    gravityOverride {z=-980}
    physicsBodyDefinitions [{
      sphereCollisionRadius=2
      boundBone {boneName=dyn_thigh_l_tube_02}
      boxExtents=<20 20 20>
      constraintSetup {angularLimitsMax=<1.5 4 5>, angularLimitsMin=<-1.5 -4 -15>, angularTarget=<0 0 0>}
      localJointOffset=<5 0 -1>
     }]
    relativeSpaceBone {boneName=root}
    simSpaceSettings {maxAngularAcceleration=500, maxAngularVelocity=50}
   }
   n17 = Anim Dynamics - Bone: dyn_thigh_l_rope_01!(ComponentPose=@n16, Alpha=1) -> (Pose) {
    angularDampingOverride=0.999
    angularSpringConstant=5000
    bAngularSpring=true
    bOverrideAngularDamping=true
    bUseGravityOverride=true
    gravityScale=0.3
    linearSpringConstant=800
    simulationSpace=RootRelative
    alphaBoolBlend {blendOption=Linear}
    boundBone {boneName=dyn_thigh_l_rope_01}
    componentAppliedLinearAccClamp=<2000 2000 2000>
    componentLinearAccScale=<0.4 0.4 0.4>
    componentLinearVelScale=<0.45 0.45 0.45>
    gravityOverride {z=-980}
    physicsBodyDefinitions [{
      sphereCollisionRadius=2
      boundBone {boneName=dyn_thigh_l_rope_01}
      boxExtents=<13 13 13>
      constraintSetup {angularLimitsMax=<1.5 4 0>, angularLimitsMin=<-1.5 -4 -15>, angularTarget=<0 0 0>}
      localJointOffset=<3 1 -1>
     }]
    relativeSpaceBone {boneName=root}
    simSpaceSettings {maxAngularAcceleration=500, maxAngularVelocity=50}
   }
   n18 = Anim Dynamics - Bone: dyn_thigh_l_rope_02!(ComponentPose=@n17, Alpha=1) -> (Pose) {
    angularDampingOverride=0.999
    angularSpringConstant=5000
    bAngularSpring=true
    bOverrideAngularDamping=true
    bUseGravityOverride=true
    gravityScale=0.3
    linearSpringConstant=800
    simulationSpace=RootRelative
    alphaBoolBlend {blendOption=Linear}
    boundBone {boneName=dyn_thigh_l_rope_02}
    componentAppliedLinearAccClamp=<2000 2000 2000>
    componentLinearAccScale=<0.4 0.4 0.4>
    componentLinearVelScale=<0.45 0.45 0.45>
    gravityOverride {z=-980}
    physicsBodyDefinitions [{
      sphereCollisionRadius=2
      boundBone {boneName=dyn_thigh_l_rope_02}
      boxExtents=<21 21 21>
      constraintSetup {angularLimitsMax=<1.5 4 0>, angularLimitsMin=<-1.5 -4 -15>, angularTarget=<0 0 0>}
      localJointOffset=<3 1 -1>
     }]
    relativeSpaceBone {boneName=root}
    simSpaceSettings {maxAngularAcceleration=500, maxAngularVelocity=50}
   }
   n19 = Anim Dynamics - Bone: thigh_volume_lt_l1_01_rope_01!(ComponentPose=@n18, Alpha=1) -> (Pose) {
    angularDampingOverride=0.999
    angularSpringConstant=5000
    bAngularSpring=true
    bOverrideAngularDamping=true
    bUseGravityOverride=true
    gravityScale=0.3
    linearSpringConstant=800
    simulationSpace=RootRelative
    alphaBoolBlend {blendOption=Linear}
    boundBone {boneName=thigh_volume_lt_l1_01_rope_01}
    componentAppliedLinearAccClamp=<1000 1000 1000>
    componentLinearAccScale=<0.2 0.2 0.2>
    componentLinearVelScale=<0.25 0.25 0.25>
    gravityOverride {z=-980}
    physicsBodyDefinitions [{
      sphereCollisionRadius=2
      boundBone {boneName=thigh_volume_lt_l1_01_rope_01}
      boxExtents=<8 8 8>
      constraintSetup {twistAxis=AxisZ, angularLimitsMax=<10 0 0>, angularLimitsMin=<-10 -15 0>, angularTarget=<0 0 0>}
      localJointOffset=<3 -1 -1>
     }]
    relativeSpaceBone {boneName=root}
    simSpaceSettings {maxAngularAcceleration=500, maxAngularVelocity=50}
   }
   n20 = Anim Dynamics - Bone: thigh_volume_lt_l1_01_rope_02!(ComponentPose=@n19, Alpha=1) -> (Pose) {
    angularDampingOverride=0.999
    angularSpringConstant=5000
    bAngularSpring=true
    bOverrideAngularDamping=true
    bUseGravityOverride=true
    gravityScale=0.3
    linearSpringConstant=800
    simulationSpace=RootRelative
    alphaBoolBlend {blendOption=Linear}
    boundBone {boneName=thigh_volume_lt_l1_01_rope_02}
    componentAppliedLinearAccClamp=<1000 1000 1000>
    componentLinearAccScale=<0.2 0.2 0.2>
    componentLinearVelScale=<0.25 0.25 0.25>
    gravityOverride {z=-980}
    physicsBodyDefinitions [{
      sphereCollisionRadius=2
      boundBone {boneName=thigh_volume_lt_l1_01_rope_02}
      boxExtents=<10 10 10>
      constraintSetup {angularLimitsMax=<1.5 10 15>, angularLimitsMin=<-1.5 -4 -10>, angularTarget=<0 0 0>}
      localJointOffset=<3 0 0>
     }]
    relativeSpaceBone {boneName=root}
    simSpaceSettings {maxAngularAcceleration=500, maxAngularVelocity=50}
   }
   n21 = Control Rig(Source=Component To Local(ComponentPose=@n20)) {
    alphaBoolBlend {blendOption=Linear}
    controlRigAssetReference {blueprintRigClass=&/F_MED_DemureIgloo_Aim_CtrlRig.F_MED_DemureIgloo_Aim_CtrlRig_C}
   }
   Output Pose(Result=@n21) {layerGroup=None}
   n22 = Control Rig(Source=Copy Pose From Mesh() {bUseAttachedParent=true}) {
    alphaBoolBlend {blendOption=Linear}
    controlRigAssetReference {blueprintRigClass=&/F_MED_DemureIgloo_CtrlRig.F_MED_DemureIgloo_CtrlRig_C}
   }
   Save cached pose 'StartPose'(Pose=@n22) {cacheName=StartPose}
  ]}
 EventGraph {nodes [Event BlueprintUpdateAnimation()]}
]
variables []
