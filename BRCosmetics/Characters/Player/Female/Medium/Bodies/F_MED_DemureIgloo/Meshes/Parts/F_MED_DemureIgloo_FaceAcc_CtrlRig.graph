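# Control Rig for the DemureIgloo female character's face accessory part, procedurally animating a radar bone chain attached to the upper spine.
# Noise blends slerped rotation offsets on the dyn_radar bones, with noise speed randomized, affected by targeting, and zeroed during jumping, falling, emotes or random pauses.
# Skydiving sets an alpha that replaces the noise with fixed blend values.
# A final spherical pose reader on the left upper arm weights corrective radar transforms.
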ControlRigBlueprint /BRCosmetics/Characters/Player/Female/Medium/Bodies/F_MED_DemureIgloo/Meshes/Parts/F_MED_DemureIgloo_FaceAcc_CtrlRig
graphs [
 ContainedGraph {nodes [
   n1 = RigVMFunctionEntryNode()
   n2 = &RigVMFunction_MathBoolOr(A=@n1.A, B=@n1.B)
   RigVMFunctionReturnNode(Result=&RigVMFunction_MathBoolOr(B=@n1.D, A=&RigVMFunction_MathBoolOr(B=@n1.C, A=@n2.Result).Result).Result)
  ]}
 ContainedGraph {nodes [
   n1 = RigVMFunctionEntryNode()
   n2 = &RigVMFunction_MathBoolOr(A=@n1.A, B=@n1.B)
   RigVMFunctionReturnNode(Result=&RigVMFunction_MathBoolOr(B=@n1.D, A=&RigVMFunction_MathBoolOr(B=@n1.C, A=@n2.Result).Result).Result)
  ]}
 ContainedGraph {nodes [
   n1 = RigVMFunctionEntryNode()
   SequenceAggregate = &RigVMFunction_Sequence(ExecuteContext=@n1.ExecuteContext)
   SequenceAggregate_1 = &RigVMFunction_Sequence(ExecuteContext=@SequenceAggregate.B)
   SequenceAggregate_2 = &RigVMFunction_Sequence(ExecuteContext=@SequenceAggregate_1.B)
   SequenceAggregate_3 = &RigVMFunction_Sequence(ExecuteContext=@SequenceAggregate_2.B)
   RigVMFunctionReturnNode(A=@SequenceAggregate.A, B=@SequenceAggregate_1.A, C=@SequenceAggregate_2.A, D=@SequenceAggregate_3.A, E=@SequenceAggregate_3.B)
  ]}
 ContainedGraph {nodes [
   n1 = RigVMFunctionEntryNode()
   RigUnit_GetTransform = &RigUnit_GetTransform(Space=LocalSpace, bInitial=false, Item=@n1.Joint)
   n2 = &RigVMFunction_MathQuaternionToEuler(RotationOrder=ZYX, Value=@RigUnit_GetTransform.Transform.Rotation)
   > RigUnit_SetTransform = &RigUnit_SetTransform(
    Space=LocalSpace
    Value=(Rotation=(X=0.0132,Y=0.4099,Z=-0.0222,W=0.9118),Translation=(X=23.6045,Y=-6.37,Z=140.1115),Scale3D=(X=1,Y=1,Z=1))
    bInitial=false
    Item=@n1.Joint
    Weight=@n1.Weight
    Value.Translation.Z=&RigVMFunction_MathFloatAdd(B=@n1.Translate Z, A=@RigUnit_GetTransform.Transform.Translation.Z) {template="Add::Execute(in A,in B,out Result)"}.Result
    Value.Scale3D.Z=&RigVMFunction_MathFloatAdd(B=@n1.Scale Z, A=@RigUnit_GetTransform.Transform.Scale3D.Z) {template="Add::Execute(in A,in B,out Result)"}.Result
    Value.Translation.Y=&RigVMFunction_MathFloatAdd(B=@n1.Translate Y, A=@RigUnit_GetTransform.Transform.Translation.Y) {template="Add::Execute(in A,in B,out Result)"}.Result
    Value.Scale3D.Y=&RigVMFunction_MathFloatAdd(B=@n1.Scale Y, A=@RigUnit_GetTransform.Transform.Scale3D.Y) {template="Add::Execute(in A,in B,out Result)"}.Result
    Value.Translation.X=&RigVMFunction_MathFloatAdd(B=@n1.Translate X, A=@RigUnit_GetTransform.Transform.Translation.X) {template="Add::Execute(in A,in B,out Result)"}.Result
    Value.Scale3D.X=&RigVMFunction_MathFloatAdd(B=@n1.Scale X, A=@RigUnit_GetTransform.Transform.Scale3D.X) {template="Add::Execute(in A,in B,out Result)"}.Result
    Value.Rotation=&RigVMFunction_MathQuaternionFromEuler(
     Euler=(X=0,Y=0,Z=0)
     Euler.X=&RigVMFunction_MathFloatAdd(B=@n1.Rotate X, A=@n2.Result.X) {template="Add::Execute(in A,in B,out Result)"}.Result
     Euler.Y=&RigVMFunction_MathFloatAdd(B=@n1.Rotate Y, A=@n2.Result.Y) {template="Add::Execute(in A,in B,out Result)"}.Result
     Euler.Z=&RigVMFunction_MathFloatAdd(B=@n1.Rotate Z, A=@n2.Result.Z) {template="Add::Execute(in A,in B,out Result)"}.Result
    ).Result
   )
   > RigVMFunctionReturnNode()
  ]}
 RigVMFunctionLibrary {nodes [Set Transforms for Bone = RigVMCollapseNode(
    ExecuteContext=()
    Joint=()
    Rotate X=0
    Rotate Y=0
    Rotate Z=0
    Scale X=0
    Scale Y=0
    Scale Z=0
    Translate X=0
    Translate Y=0
    Translate Z=0
    Weight=1
   )]}
 RigVMModel {nodes [
   RigUnit_BeginExecution = &RigUnit_BeginExecution()
   SequenceAggregate = RigVMAggregateNode(ExecuteContext=@RigUnit_BeginExecution.ExecutePin)
   MathBoolOr = RigVMAggregateNode(B=RigVMVariableNode() {variable=isSkyDiving}.Value)
   RigVMVariableNode(
    Value=RigVMDispatchNode(False=0, True=1, Condition=@MathBoolOr.Result) {template="DISPATCH_RigVMDispatch_If(in Condition,in True,in False,out Result)"}.Result
    ExecuteContext=@SequenceAggregate.A
   ) {variable=randomMotionAlpha}
   Greater = /Script/RigVM.RigVMFunction_MathDoubleGreater(
    B=0.35
    A=/Script/RigVM.RigVMFunction_RandomFloat(Duration=1, Maximum=1, Minimum=-1, Seed=15) {is=struct RigVMUnitNode, template="Random::Execute(in Seed,in Minimum,in Maximum,in Duration,out Result)"}.Result
   ) {is=struct RigVMUnitNode, template="Greater::Execute(in A,in B,out Result)"}
   MathBoolOr_1 = RigVMAggregateNode(
    C=&RigVMFunction_MathBoolNot(Value=@Greater.Result).Result
    A=RigVMVariableNode() {variable=isJumping}.Value
    B=RigVMVariableNode() {variable=isFalling}.Value
    D=RigVMVariableNode() {variable=isEmote}.Value
   )
   n1 = RigVMDispatchNode(
    True=0
    False=RigVMDispatchNode(
     True=1
     False=&RigVMFunction_RandomFloat(Duration=1, Maximum=1.1, Minimum=0, Seed=50) {template="Random::Execute(in Seed,in Minimum,in Maximum,in Duration,out Result)"}.Result
     Condition=RigVMVariableNode() {variable=isTargeting}.Value
    ) {template="DISPATCH_RigVMDispatch_If(in Condition,in True,in False,out Result)"}.Result
    Condition=@MathBoolOr_1.Result
   ) {template="DISPATCH_RigVMDispatch_If(in Condition,in True,in False,out Result)"}
   Noise = /Script/RigVM.RigVMFunction_NoiseFloat(Frequency=1, Maximum=1, Minimum=0, Value=1.5, Speed=@n1.Result) {is=struct RigVMUnitNode, template="Noise::Execute(in Value,in Speed,in Frequency,in Minimum,in Maximum,out Result)"}
   n2 = RigVMVariableNode() {variable=randomMotionAlpha}
   Interpolate = /Script/RigVM.RigVMFunction_MathFloatLerp(B=0.5, A=@Noise.Result, T=@n2.Value) {is=struct RigVMUnitNode, template="Interpolate::Execute(in A,in B,in T,out Result)"}
   > &RigUnit_ModifyTransforms(
    ItemToModify=((Item=(Type=Bone,Name="dyn_radar_01"),Transform=(Rotation=(X=0,Y=0,Z=0,W=1),Translation=(X=0,Y=0,Z=0),Scale3D=(X=1,Y=1,Z=1))),(Item=(Type=Bone,Name="dyn_radar_MSTR"),Transform=(Rotation=(X=0,Y=0,Z=0,W=1),Translation=(X=0,Y=0,Z=0),Scale3D=(X=1,Y=1,Z=1))))
    Weight=1
    ItemToModify.0.Transform.Rotation=/Script/RigVM.RigVMFunction_MathQuaternionSlerp(A=(X=0,Y=0,Z=-0.0436,W=0.999), B=(X=0,Y=0,Z=0,W=1), T=@Interpolate.Result) {is=struct RigVMUnitNode, template="Interpolate::Execute(in A,in B,in T,out Result)"}.Result
    ItemToModify.1.Transform.Rotation=&RigVMFunction_MathQuaternionSlerp(A=(X=0,Y=0,Z=-0.061,W=0.9981), B=(X=0,Y=0,Z=0,W=1), T=@Interpolate.Result) {template="Interpolate::Execute(in A,in B,in T,out Result)"}.Result
   )
   > &RigUnit_ModifyTransforms(
    ItemToModify=((Item=(Type=Bone,Name="dyn_radar_02"),Transform=(Rotation=(X=0,Y=0,Z=0,W=1),Translation=(X=0,Y=0,Z=0),Scale3D=(X=1,Y=1,Z=1))),(Item=(Type=Bone,Name="dyn_radar_02"),Transform=(Rotation=(X=0,Y=0,Z=0,W=1),Translation=(X=0,Y=0,Z=0),Scale3D=(X=1,Y=1,Z=1))))
    Weight=1
    ItemToModify.0.Transform.Rotation=&RigVMFunction_MathQuaternionSlerp(A=(X=0,Y=-0.061,Z=0,W=0.9981), B=(X=0,Y=-0.0175,Z=0,W=0.9998), T=@Interpolate.Result) {template="Interpolate::Execute(in A,in B,in T,out Result)"}.Result
    ItemToModify.1.Transform.Rotation=&RigVMFunction_MathQuaternionSlerp(A=(X=0,Y=-0.0872,Z=0,W=0.9962), B=(X=0,Y=0.0175,Z=0,W=0.9998), T=@Interpolate.Result) {template="Interpolate::Execute(in A,in B,in T,out Result)"}.Result
   )
   n3 = &RigVMFunction_RandomFloat(Duration=1, Maximum=1.2, Minimum=0, Seed=50) {template="Random::Execute(in Seed,in Minimum,in Maximum,in Duration,out Result)"}
   n4 = &RigVMFunction_MathFloatLerp(
    B=0.5
    A=&RigVMFunction_NoiseFloat(
     Frequency=0.5
     Maximum=1
     Minimum=0
     Value=-7
     Speed=RigVMDispatchNode(True=0, Condition=@MathBoolOr_1.Result, False=@n3.Result) {template="DISPATCH_RigVMDispatch_If(in Condition,in True,in False,out Result)"}.Result
    ) {template="Noise::Execute(in Value,in Speed,in Frequency,in Minimum,in Maximum,out Result)"}.Result
    T=@n2.Value
   ) {template="Interpolate::Execute(in A,in B,in T,out Result)"}
   > &RigUnit_ModifyTransforms(
    ItemToModify=((Item=(Type=Bone,Name="dyn_radar_03"),Transform=(Rotation=(X=0,Y=0,Z=0,W=1),Translation=(X=0,Y=0,Z=0),Scale3D=(X=1,Y=1,Z=1))),(Item=(Type=Bone,Name="dyn_radar_02"),Transform=(Rotation=(X=0,Y=0,Z=0,W=1),Translation=(X=0,Y=0,Z=0),Scale3D=(X=1,Y=1,Z=1))))
    Weight=1
    ItemToModify.0.Transform.Rotation=&RigVMFunction_MathQuaternionSlerp(A=(X=0,Y=0.061,Z=0,W=0.9981), B=(X=0,Y=0,Z=0,W=1), T=@n4.Result) {template="Interpolate::Execute(in A,in B,in T,out Result)"}.Result
    ItemToModify.1.Transform.Rotation=&RigVMFunction_MathQuaternionSlerp(A=(X=0,Y=-0.061,Z=0,W=0.9981), B=(X=0,Y=0,Z=0,W=1), T=@n4.Result) {template="Interpolate::Execute(in A,in B,in T,out Result)"}.Result
   )
   > &RigUnit_ModifyTransforms(
    ItemToModify=((Item=(Type=Bone,Name="dyn_radar_04"),Transform=(Rotation=(X=0,Y=0,Z=0,W=1),Translation=(X=0,Y=0,Z=0),Scale3D=(X=1,Y=1,Z=1))),(Item=(Type=Bone,Name="dyn_radar_04"),Transform=(Rotation=(X=0,Y=0,Z=0,W=1),Translation=(X=0,Y=0,Z=0),Scale3D=(X=1,Y=1,Z=1))))
    Weight=1
    ItemToModify.0.Transform.Rotation=&RigVMFunction_MathQuaternionSlerp(A=(X=-0.0015,Y=-0.0174,Z=-0.0871,W=0.996), B=(X=0,Y=-0.0872,Z=0,W=0.9962), T=@n4.Result) {template="Interpolate::Execute(in A,in B,in T,out Result)"}.Result
    ItemToModify.1.Transform.Rotation=&RigVMFunction_MathQuaternionSlerp(A=(X=0.0151,Y=0.0858,Z=-0.173,W=0.9811), B=(X=0,Y=-0.0436,Z=0,W=0.999), T=@n4.Result) {template="Interpolate::Execute(in A,in B,in T,out Result)"}.Result
   )
   n5 = &RigVMFunction_MathFloatLerp(
    B=0.35
    A=&RigVMFunction_NoiseFloat(Frequency=0.1, Maximum=1, Minimum=0.5, Value=5, Speed=@n1.Result) {template="Noise::Execute(in Value,in Speed,in Frequency,in Minimum,in Maximum,out Result)"}.Result
    T=@n2.Value
   ) {template="Interpolate::Execute(in A,in B,in T,out Result)"}
   > &RigUnit_ModifyTransforms(
    ItemToModify=((Item=(Type=Bone,Name="dyn_radar_04"),Transform=(Rotation=(X=0,Y=0,Z=0,W=1),Translation=(X=0,Y=0,Z=0),Scale3D=(X=1,Y=1,Z=1))),(Item=(Type=Bone,Name="dyn_radar_04"),Transform=(Rotation=(X=0,Y=0,Z=0,W=1),Translation=(X=0,Y=0,Z=0),Scale3D=(X=1,Y=1,Z=1))))
    Weight=1
    ItemToModify.0.Transform.Rotation=&RigVMFunction_MathQuaternionSlerp(A=(X=0.0114,Y=0.0864,Z=-0.13,W=0.9877), B=(X=0,Y=0,Z=-0.0175,W=0.9998), T=@n5.Result) {template="Interpolate::Execute(in A,in B,in T,out Result)"}.Result
   )
   > &RigUnit_ModifyTransforms(
    ItemToModify="((Item=(Type=Bone,Name=\"dyn_radar_MSTR\"),Transform=(Rotation=(X=0,Y=0,Z=0,W=1),Translation=(X=0,Y=0,Z=0),Scale3D=(X=1,Y=1,Z=1))),(Item=(Type=Bone,Name=\"dyn_radar_MSTR\"),Transform=(Rotation=(X=0,Y=0,Z=0,W=1),Translation=(X=0,Y=0,Z=0),Scale3D=(X=1,Y=1,Z=1))) ...(truncated)"
    Weight=1
    ItemToModify.0.Transform.Rotation=&RigVMFunction_MathQuaternionSlerp(A=(X=0,Y=0,Z=-0.1736,W=0.9848), B=(X=0,Y=0,Z=0.1305,W=0.9914), T=@n5.Result) {template="Interpolate::Execute(in A,in B,in T,out Result)"}.Result
   )
   > n6 = &RigUnit_SphericalPoseReader(
    ActiveRegionScaleFactors=(PositiveWidth=0.1,NegativeWidth=1,PositiveHeight=0.1,NegativeHeight=2)
    ActiveRegionSize=0.1
    Debug=(bDrawDebug=false,DebugScale=25,DebugSegments=20,DebugThickness=0.25)
    DriverAxis=(X=1,Y=0,Z=0)
    DriverItem=(Type=Bone,Name="upperarm_l")
    FalloffRegionScaleFactors=(PositiveWidth=0.7,NegativeWidth=1,PositiveHeight=0.8,NegativeHeight=1.5)
    FalloffSize=0.5
    FlipWidthScaling=true
    OptionalParentItem=(Type=Bone,Name="spine_05")
    RotationOffset=(X=0,Y=30,Z=-10)
   )
   > &RigUnit_ModifyTransforms(
    ItemToModify="((Item=(Type=Bone,Name=\"dyn_radar_MSTR\"),Transform=(Rotation=(X=0,Y=0,Z=0.7071,W=0.7071),Translation=(X=0,Y=0,Z=0),Scale3D=(X=1,Y=1,Z=1))),(Item=(Type=Bone,Name=\"dyn_radar_01\"),Transform=(Rotation=(X=0,Y=0,Z=-0.5736,W=0.8192),Translation=(X=0,Y=0,Z=0),Scal ...(truncated)"
    Weight=@n6.OutputParam
   )
  ]}
]
hierarchy [
 root {loc=<0 0 0>, rot=<0 0 0 1>}
 pelvis {parent=root, loc=<0 -0.5301 98.0571>, rot=<0.0127 0.707 -0.0127 -0.707>}
 spine_01 {parent=pelvis, loc=<6.1208 -0.9328 0>, rot=<0 0 0.034 -0.9994>}
 spine_02 {parent=spine_01, loc=<6.5997 -0.2095 0>, rot=<0 0 0.0238 -0.9997>}
 spine_03 {parent=spine_02, loc=<6.4914 -0.1237 0>, rot=<0 0 0.0149 -0.9999>}
 spine_04 {parent=spine_03, loc=<6.9598 0.0794 0>, rot=<0 0 -0.0132 -0.9999>}
 spine_05 {parent=spine_04, loc=<6.9987 0 0>, rot=<0 0 -0.0015 -1>}
 neck_01 {parent=spine_05, loc=<11.3138 0.1529 0>, rot=<0 0 -0.2003 -0.9797>}
 neck_02 {parent=neck_01, loc=<4.5627 0 0>, rot=<0 0 0.0504 -0.9987>}
 head {parent=neck_02, loc=<5.3914 0 0>, rot=<0 0 0.115 -0.9934>}
 def_radar_base_MSTR {parent=spine_05, loc=<11.9799 0.4548 -6.7601>, rot=<0.0525 -0.7742 -0.0079 -0.6307>}
 def_radar_base {parent=def_radar_base_MSTR, loc=<4.2731 -0.453 7.3087>, rot=<0.0099 0.1225 0.0423 -0.9915>}
 dyn_radar_MSTR {parent=def_radar_base, loc=<3.4396 0.4385 -0.1379>, rot=<-0.2216 0.0799 -0.3967 -0.8872>}
 dyn_radar_01 {parent=dyn_radar_MSTR, loc=<3.6186 -0.1236 0.0735>, rot=<-0.0144 0.024 -0.2096 -0.9774>}
 dyn_radar_02 {parent=dyn_radar_01, loc=<4.965 0 0>, rot=<0 0.1937 0 -0.9811>}
 dyn_radar_03 {parent=dyn_radar_02, loc=<3.7013 0 0>, rot=<-0.0125 -0.4311 -0.0261 -0.9018>}
 dyn_radar_04 {parent=dyn_radar_03, loc=<3.9154 -0.0791 -0.1506>, rot=<0.0861 -0.0086 -0.189 -0.9782>}
 dyn_radar_04_rope_a {parent=dyn_radar_04, loc=<2.6336 0.5174 -2.1744>, rot=<0 0 0 1>}
 dyn_radar_04_rope_b {parent=dyn_radar_04, loc=<-0.3732 0.5174 -1.2779>, rot=<0 0 0 1>}
 clavicle_l {parent=spine_05, loc=<5.4831 -0.1053 -3.2946>, rot=<0.0465 -0.7289 -0.009 -0.6829>}
 upperarm_l {parent=clavicle_l, loc=<9.8658 -1.2136 -0.4105>, rot=<0.0164 -0.3245 -0.0269 -0.9454>}
 lowerarm_l {parent=upperarm_l, loc=<26.465 -0.8345 0.259>, rot=<0.0036 0.0161 -0.2298 -0.9731>}
 clavicle_r {parent=spine_05, loc=<5.4831 -0.1053 3.2946>, rot=<0.7289 0.0465 -0.6829 0.009>}
 upperarm_r {parent=clavicle_r, loc=<-9.8658 1.2136 0.4105>, rot=<0.0164 -0.3245 -0.0269 -0.9454>}
 lowerarm_r {parent=upperarm_r, loc=<-26.465 0.8345 -0.259>, rot=<0.0036 0.0161 -0.2298 -0.9731>}
]
variables [
 isEmote {type=bool, default=False}
 isFalling {type=bool, default=False}
 isJumping {type=bool, default=False}
 isSkyDiving {type=bool, default=False}
 isTargeting {type=bool, default=False}
 randomMotionAlpha {type=double, default=0}
]
