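# Control Rig for the DemureIgloo female character's face accessory part, posing a radar attached to the upper chest near the left shoulder.
# The dyn_radar_02 bone follows head rotation with a local-space offset.
# A spherical pose reader on upperarm_l, relative to spine_05, weights corrective constraints to spine_05 and the head (X/Z only), plus transform tweaks on radar bones.
# The reader has debug drawing enabled, and the isTargeting variable is unused.
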ControlRigBlueprint /BRCosmetics/Characters/Player/Female/Medium/Bodies/F_MED_DemureIgloo/Meshes/Parts/F_MED_DemureIgloo_FaceAcc_CtrlRig1
graphs [
 RigVMFunctionLibrary {links [], nodes []}
 RigVMModel {nodes [
   RigUnit_BeginExecution = &RigUnit_BeginExecution()
   &RigUnit_RotationConstraintLocalSpaceOffset(
    AdvancedSettings=(InterpolationType=Average,RotationOrderForFilter=XZY)
    Child=(Type=Bone,Name="dyn_radar_02")
    Filter=(bX=True,bY=true,bZ=True)
    Parents=((Item=(Type=Bone,Name="head"),Weight=1))
    Weight=1
    ExecutePin=@RigUnit_BeginExecution.ExecutePin
   )
   > n1 = &RigUnit_SphericalPoseReader(
    ActiveRegionScaleFactors=(PositiveWidth=0.1,NegativeWidth=1,PositiveHeight=1,NegativeHeight=0.1)
    ActiveRegionSize=0.1
    Debug=(bDrawDebug=true,DebugScale=25,DebugSegments=20,DebugThickness=0.25)
    DriverAxis=(X=1,Y=0,Z=0)
    DriverItem=(Type=Bone,Name="upperarm_l")
    FalloffRegionScaleFactors=(PositiveWidth=0.5,NegativeWidth=1,PositiveHeight=1,NegativeHeight=0.5)
    FalloffSize=0.6
    OptionalParentItem=(Type=Bone,Name="spine_05")
    RotationOffset=(X=0,Y=0,Z=0)
   )
   > &RigUnit_RotationConstraintLocalSpaceOffset(
    AdvancedSettings=(InterpolationType=Average,RotationOrderForFilter=XZY)
    Child=(Type=Bone,Name="dyn_radar_02")
    Filter=(bX=True,bY=true,bZ=True)
    Parents=((Item=(Type=Bone,Name="spine_05"),Weight=1))
    Weight=@n1.OutputParam
   )
   > &RigUnit_RotationConstraintLocalSpaceOffset(
    AdvancedSettings=(InterpolationType=Average,RotationOrderForFilter=XZY)
    Child=(Type=Bone,Name="dyn_radar_02")
    Filter=(bX=True,bY=false,bZ=True)
    Parents=((Item=(Type=Bone,Name="head"),Weight=1))
    Weight=@n1.OutputParam
   )
   > &RigUnit_ModifyTransforms(
    ItemToModify="((Item=(Type=Bone,Name=\"dyn_radar_02\"),Transform=(Rotation=(X=0,Y=-0.7071,Z=0,W=0.7071),Translation=(X=0,Y=0,Z=0),Scale3D=(X=1,Y=1,Z=1))),(Item=(Type=Bone,Name=\"dyn_radar_01\"),Transform=(Rotation=(X=0,Y=0,Z=0.0872,W=0.9962),Translation=(X=0,Y=0,Z=0),Scale3 ...(truncated)"
    Weight=@n1.OutputParam
   )
  ]}
]
hierarchy [
 root {loc=<0 0 0>, rot=<0 0 0 1>}
 pelvis {parent=root, loc=<0 -0.5301 98.0571>, rot=<0.0127 0.707 -0.0127 -0.707>}
 spine_01 {parent=pelvis, loc=<6.1208 -0.9328 0>, rot=<0 0 0.034 -0.9994>}
 spine_02 {parent=spine_01, loc=<6.5997 -0.2095 0>, rot=<0 0 0.0238 -0.9997>}
 spine_03 {parent=spine_02, loc=<6.4914 -0.1237 0>, rot=<0 0 0.0149 -0.9999>}
 spine_04 {parent=spine_03, loc=<6.9598 0.0794 0>, rot=<0 0 -0.0132 -0.9999>}
 spine_05 {parent=spine_04, loc=<6.9987 0 0>, rot=<0 0 -0.0015 -1>}
 neck_01 {parent=spine_05, loc=<11.3138 0.1529 0>, rot=<0 0 -0.2003 -0.9797>}
 neck_02 {parent=neck_01, loc=<4.5627 0 0>, rot=<0 0 0.0504 -0.9987>}
 head {parent=neck_02, loc=<5.3914 0 0>, rot=<0 0 0.115 -0.9934>}
 clavicle_l {parent=spine_05, loc=<5.4831 -0.1053 -3.2946>, rot=<0.0465 -0.7289 -0.009 -0.6829>}
 upperarm_l {parent=clavicle_l, loc=<9.8658 -1.2136 -0.4105>, rot=<0.0164 -0.3245 -0.0269 -0.9454>}
 lowerarm_l {parent=upperarm_l, loc=<26.465 -0.8345 0.259>, rot=<0.0036 0.0161 -0.2298 -0.9731>}
 clavicle_r {parent=spine_05, loc=<5.4831 -0.1053 3.2946>, rot=<0.7289 0.0465 -0.6829 0.009>}
 upperarm_r {parent=clavicle_r, loc=<-9.8658 1.2136 0.4105>, rot=<0.0164 -0.3245 -0.0269 -0.9454>}
 lowerarm_r {parent=upperarm_r, loc=<-26.465 0.8345 -0.259>, rot=<0.0036 0.0161 -0.2298 -0.9731>}
 def_radar_base_MSTR {parent=spine_05, loc=<11.9799 0.4548 -6.7601>, rot=<0.0525 -0.7742 -0.0079 -0.6307>}
 def_radar_base {parent=def_radar_base_MSTR, loc=<4.2731 -0.453 7.3087>, rot=<0.0099 0.1225 0.0423 -0.9915>}
 dyn_radar_MSTR {parent=def_radar_base, loc=<3.4396 0.4385 -0.1379>, rot=<-0.2216 0.0799 -0.3967 -0.8872>}
 dyn_radar_01 {parent=dyn_radar_MSTR, loc=<3.6186 -0.1236 0.0735>, rot=<-0.0144 0.024 -0.2096 -0.9774>}
 dyn_radar_02 {parent=dyn_radar_01, loc=<4.965 0 0>, rot=<0 0.1937 0 -0.9811>}
 dyn_radar_03 {parent=dyn_radar_02, loc=<3.7013 0 0>, rot=<-0.0125 -0.4311 -0.0261 -0.9018>}
 dyn_radar_04 {parent=dyn_radar_03, loc=<3.9154 -0.0791 -0.1506>, rot=<0.0861 -0.0086 -0.189 -0.9782>}
 dyn_radar_04_rope_a {parent=dyn_radar_04, loc=<2.6336 0.5174 -2.1744>, rot=<0 0 0 1>}
 dyn_radar_04_rope_b {parent=dyn_radar_04, loc=<-0.3732 0.5174 -1.2779>, rot=<0 0 0 1>}
]
variables [isTargeting {type=bool, default=False}]
