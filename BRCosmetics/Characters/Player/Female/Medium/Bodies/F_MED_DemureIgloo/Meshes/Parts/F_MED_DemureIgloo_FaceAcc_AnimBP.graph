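# Animation Blueprint for the face accessory part of the DemureIgloo female character cosmetic, posing it from the attached parent mesh.
# The copied pose feeds two Control Rigs, with CtrlRig1 blended in during targeting or melee and the other rig, which also receives skydive, jump, fall and emote states, used otherwise.
# Two root-relative Anim Dynamics nodes then add spring-driven, angle-limited sway to the radar rope bones.
# The event graph updates the targeting-or-melee flag.
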AnimBlueprint /BRCosmetics/Characters/Player/Female/Medium/Bodies/F_MED_DemureIgloo/Meshes/Parts/F_MED_DemureIgloo_FaceAcc_AnimBP
target_skeleton=&/F_MED_DemureIgloo_FaceAcc_Skeleton
graphs [
 AnimGraph {nodes [
   n1 = Control Rig!(Source=Use cached pose 'StartPose'(), Alpha, isTargeting:bool=Get bIsTargeting()) -> (Pose) {
    alphaBoolBlend {blendOption=Linear}
    controlRigAssetReference {blueprintRigClass=&/F_MED_DemureIgloo_FaceAcc_CtrlRig1.F_MED_DemureIgloo_FaceAcc_CtrlRig1_C}
   }
   n2 = Control Rig!(
    Source=Use cached pose 'StartPose'()
    Alpha
    isTargeting=Get bIsTargeting()
    isSkyDiving:bool=Get bIsSkyDiving()
    isJumping:bool=Get bIsJumping()
    isFalling:bool=Get bIsFalling()
    isEmote:bool=Get bIsPlayingEmote()
   ) -> (Pose) {
    alphaBoolBlend {blendOption=Linear}
    controlRigAssetReference {blueprintRigClass=&/F_MED_DemureIgloo_FaceAcc_CtrlRig.F_MED_DemureIgloo_FaceAcc_CtrlRig_C}
   }
   n3 = Get Targeting or Melee Anim() -> (Targeting or Melee Anim:bool)
   n4 = Anim Dynamics - Bone: dyn_radar_04_rope_a!(
    ComponentPose=Local To Component(LocalPose=Blend Poses by bool(BlendPose_0=@n1, BlendPose_1=@n2, BlendTime_0=0.3, BlendTime_1=1, bActiveValue=@n3) {blendType=Linear})
    Alpha=1
    GravityOverride:Vector="0,0,-980"
   ) -> (Pose) {
    angularDampingOverride=1
    angularSpringConstant=800
    bOverrideAngularDamping=true
    bUseGravityOverride=true
    gravityScale=0.3
    linearSpringConstant=800
    simulationSpace=RootRelative
    alphaBoolBlend {blendOption=Linear}
    boundBone {boneName=dyn_radar_04_rope_a}
    componentAppliedLinearAccClamp=<2 2 2>
    gravityOverride {z=-980}
    physicsBodyDefinitions [{
      sphereCollisionRadius=2
      boundBone {boneName=dyn_radar_04_rope_a}
      boxExtents=<10 10 10>
      constraintSetup {twistAxis=AxisZ, angularLimitsMax=<15 30 3>, angularLimitsMin=<-15 -10 -3>, angularTarget=<0 0 0>}
      localJointOffset=<0.1 -0.1 -3>
     }]
    relativeSpaceBone {boneName=root}
    simSpaceSettings {maxAngularAcceleration=50, maxAngularVelocity=8, simSpaceAngularAlpha=0.2, externalAngularVelocity {x=3}}
   }
   n5 = Anim Dynamics - Bone: dyn_radar_04_rope_b!(ComponentPose=@n4, Alpha=1, GravityOverride:Vector="0,0,-980") -> (Pose) {
    angularDampingOverride=1
    angularSpringConstant=800
    bOverrideAngularDamping=true
    bUseGravityOverride=true
    gravityScale=0.3
    linearSpringConstant=800
    simulationSpace=RootRelative
    alphaBoolBlend {blendOption=Linear}
    boundBone {boneName=dyn_radar_04_rope_b}
    componentAppliedLinearAccClamp=<2 2 2>
    gravityOverride {z=-980}
    physicsBodyDefinitions [{
      sphereCollisionRadius=2
      boundBone {boneName=dyn_radar_04_rope_b}
      boxExtents=<10 10 10>
      constraintSetup {twistAxis=AxisZ, angularLimitsMax=<12 30 3>, angularLimitsMin=<-12 -10 -3>, angularTarget=<0 0 0>}
      localJointOffset=<0.1 -0.1 -4>
     }]
    relativeSpaceBone {boneName=root}
    simSpaceSettings {maxAngularAcceleration=50, maxAngularVelocity=8, simSpaceAngularAlpha=0.2, externalAngularVelocity {x=3}}
   }
   Output Pose(Result=Component To Local(ComponentPose=@n5)) {layerGroup=None}
   Save cached pose 'StartPose'(Pose=Copy Pose From Mesh() {bUseAttachedParent=true, bUseMeshPose=true}) {cacheName=StartPose}
  ]}
 EventGraph {nodes [
   Event BlueprintUpdateAnimation()
   > Set Targeting or Melee Anim(Targeting or Melee Anim:bool=OR Boolean!(A=Get bIsTargeting(), B=Get bIsPlayingMeleeAnim(), C:bool) -> (ReturnValue)) -> (Output_Get:bool)
  ]}
]
variables [Targeting or Melee Anim {type=boolean}]
